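# Control Rig for the F_MED_Blue_Jay cosmetic body that procedurally adjusts skirt bones as the thighs move.
# Per leg, spherical pose readers on the thigh weight parent constraints blending skirt B/C masters between thigh and pelvis, plus corrective transform offsets.
# A third sequence branch blends front skirt bones between left and right B chains, rotating their master by the right-thigh reader alone.
# Several left-side reader and constraint chains remain disconnected from execution.
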ControlRigBlueprint /BRCosmetics/Characters/Player/Female/Medium/Bodies/F_MED_Blue_Jay/Meshes/F_MED_Blue_Jay_CtrlRig
graphs [
 RigVMFunctionLibrary {links [], nodes []}
 RigVMModel {nodes [
   RigUnit_BeginExecution = &RigUnit_BeginExecution()
   n1 = /Script/ControlRig.RigUnit_SequenceExecution(ExecuteContext=@RigUnit_BeginExecution.ExecutePin) {is=struct RigVMUnitNode}
   n2 = &RigUnit_SphericalPoseReader(
    ActiveRegionScaleFactors=(PositiveWidth=3.5,NegativeWidth=2.5,PositiveHeight=3,NegativeHeight=1.3)
    ActiveRegionSize=0.11
    Debug=(bDrawDebug=false,DebugScale=25,DebugSegments=20,DebugThickness=0.25)
    DriverAxis=(X=-1,Y=0,Z=0)
    DriverItem=(Type=Bone,Name="thigh_l")
    FalloffRegionScaleFactors=(PositiveWidth=0.9,NegativeWidth=1,PositiveHeight=1.1,NegativeHeight=0.9)
    FalloffSize=0.5
    OptionalParentItem=(Type=Bone)
    RotationOffset=(X=110,Y=0,Z=0)
    ExecutePin=@n1.A
   )
   > &RigUnit_ParentConstraint(
    AdvancedSettings=(RotationOrderForFilter=XZY)
    Child=(Type=Bone,Name="dyn_L_skirt_B_master")
    Filter=(TranslationFilter=(bX=false,bY=false,bZ=false),RotationFilter=(bX=True,bY=True,bZ=True),ScaleFilter=(bX=True,bY=True,bZ=True))
    Parents=((Item=(Type=Bone,Name="thigh_l"),Weight=0.8),(Item=(Type=Bone,Name="pelvis"),Weight=0.2))
    Weight=@n2.OutputParam
   )
   > &RigUnit_ParentConstraint(
    AdvancedSettings=(RotationOrderForFilter=XZY)
    Child=(Type=Bone,Name="dyn_L_skirt_C_master")
    Filter=(TranslationFilter=(bX=false,bY=false,bZ=false),RotationFilter=(bX=True,bY=True,bZ=True),ScaleFilter=(bX=True,bY=True,bZ=True))
    Parents=((Item=(Type=Bone,Name="thigh_l"),Weight=0.4),(Item=(Type=Bone,Name="pelvis"),Weight=0.6))
    Weight=@n2.OutputParam
   )
   > n3 = &RigUnit_SphericalPoseReader(
    ActiveRegionScaleFactors=(PositiveWidth=3.5,NegativeWidth=0.5,PositiveHeight=3,NegativeHeight=3)
    ActiveRegionSize=0.1
    Debug=(bDrawDebug=false,DebugScale=25,DebugSegments=20,DebugThickness=0.25)
    DriverAxis=(X=-1,Y=0,Z=0)
    DriverItem=(Type=Bone,Name="thigh_l")
    FalloffRegionScaleFactors=(PositiveWidth=1,NegativeWidth=1,PositiveHeight=1,NegativeHeight=1)
    FalloffSize=0.45
    OptionalParentItem=(Type=Bone)
    RotationOffset=(X=90,Y=0,Z=0)
   )
   > &RigUnit_ModifyTransforms(
    ItemToModify=((Item=(Type=Bone,Name="dyn_L_skirt_C_master"),Transform=(Rotation=(X=0,Y=0,Z=0,W=1),Translation=(X=0,Y=0,Z=3),Scale3D=(X=1,Y=1,Z=1))))
    ItemToModify.0.Transform.Rotation=&RigVMFunction_MathQuaternionFromEuler(Euler=(X=0,Y=-15,Z=0)).Result
    Weight=@n3.OutputParam
   )
   > &RigUnit_ModifyTransforms(
    ItemToModify=((Item=(Type=Bone,Name="dyn_L_skirt_B_master"),Transform=(Rotation=(X=0,Y=0,Z=0,W=1),Translation=(X=-2,Y=0,Z=1),Scale3D=(X=1,Y=1,Z=1))))
    Weight=@n3.OutputParam
   )
   n4 = &RigUnit_SphericalPoseReader(
    ActiveRegionScaleFactors=(PositiveWidth=3.5,NegativeWidth=2,PositiveHeight=2,NegativeHeight=1.1)
    ActiveRegionSize=0.11
    Debug=(bDrawDebug=false,DebugScale=25,DebugSegments=20,DebugThickness=0.25)
    DriverAxis=(X=1,Y=0,Z=0)
    DriverItem=(Type=Bone,Name="thigh_r")
    FalloffRegionScaleFactors=(PositiveWidth=0.9,NegativeWidth=1.1,PositiveHeight=1.1,NegativeHeight=0.9)
    FalloffSize=0.47
    FlipWidthScaling=true
    OptionalParentItem=(Type=Bone)
    RotationOffset=(X=-70,Y=0,Z=15)
    ExecutePin=@n1.B
   )
   > &RigUnit_ParentConstraint(
    AdvancedSettings=(RotationOrderForFilter=XZY)
    Child=(Type=Bone,Name="dyn_R_skirt_B_master")
    Filter=(TranslationFilter=(bX=false,bY=false,bZ=false),RotationFilter=(bX=True,bY=True,bZ=True),ScaleFilter=(bX=True,bY=True,bZ=True))
    Parents=((Item=(Type=Bone,Name="thigh_r"),Weight=0.8),(Item=(Type=Bone,Name="pelvis"),Weight=0.2))
    Weight=@n4.OutputParam
   )
   > &RigUnit_ParentConstraint(
    AdvancedSettings=(RotationOrderForFilter=XZY)
    Child=(Type=Bone,Name="dyn_R_skirt_C_master")
    Filter=(TranslationFilter=(bX=false,bY=false,bZ=false),RotationFilter=(bX=True,bY=True,bZ=True),ScaleFilter=(bX=True,bY=True,bZ=True))
    Parents=((Item=(Type=Bone,Name="thigh_r"),Weight=0.4),(Item=(Type=Bone,Name="pelvis"),Weight=0.6))
    Weight=@n4.OutputParam
   )
   > n5 = &RigUnit_SphericalPoseReader(
    ActiveRegionScaleFactors=(PositiveWidth=3.5,NegativeWidth=0.5,PositiveHeight=3,NegativeHeight=3)
    ActiveRegionSize=0.1
    Debug=(bDrawDebug=false,DebugScale=25,DebugSegments=20,DebugThickness=0.25)
    DriverAxis=(X=1,Y=0,Z=0)
    DriverItem=(Type=Bone,Name="thigh_r")
    FalloffRegionScaleFactors=(PositiveWidth=1,NegativeWidth=1,PositiveHeight=1,NegativeHeight=1)
    FalloffSize=0.45
    FlipHeightScaling=false
    FlipWidthScaling=true
    OptionalParentItem=(Type=Bone)
    RotationOffset=(X=-90,Y=0,Z=0)
   )
   n6 = &RigVMFunction_MathQuaternionFromEuler(Euler=(X=0,Y=-15,Z=0))
   > &RigUnit_ModifyTransforms(
    ItemToModify=((Item=(Type=Bone,Name="dyn_R_skirt_C_master"),Transform=(Rotation=(X=0,Y=0,Z=0,W=1),Translation=(X=0,Y=0,Z=-3),Scale3D=(X=1,Y=1,Z=1))))
    ItemToModify.0.Transform.Rotation=@n6.Result
    Weight=@n5.OutputParam
   )
   > &RigUnit_ModifyTransforms(
    ItemToModify=((Item=(Type=Bone,Name="dyn_R_skirt_B_master"),Transform=(Rotation=(X=0,Y=0,Z=0,W=1),Translation=(X=2,Y=0,Z=-1),Scale3D=(X=1,Y=1,Z=1))))
    ItemToModify.0.Transform.Rotation=@n6.Result
    Weight=@n5.OutputParam
   )
   &RigUnit_ParentConstraint(
    AdvancedSettings=(RotationOrderForFilter=XZY)
    Child=(Type=Bone,Name="dyn_frontSkirt_01")
    Filter=(TranslationFilter=(bX=false,bY=false,bZ=false),RotationFilter=(bX=True,bY=True,bZ=True),ScaleFilter=(bX=True,bY=True,bZ=True))
    Parents=((Item=(Type=Bone,Name="dyn_L_skirt_B_01"),Weight=0.3),(Item=(Type=Bone,Name="dyn_R_skirt_B_01"),Weight=0.7))
    Weight=1
    ExecutePin=@n1.C
   )
   > &RigUnit_ParentConstraint(
    AdvancedSettings=(RotationOrderForFilter=XZY)
    Child=(Type=Bone,Name="dyn_frontSkirt_02")
    Filter=(TranslationFilter=(bX=false,bY=false,bZ=false),RotationFilter=(bX=True,bY=True,bZ=True),ScaleFilter=(bX=True,bY=True,bZ=True))
    Parents=((Item=(Type=Bone,Name="dyn_L_skirt_B_02"),Weight=0.5),(Item=(Type=Bone,Name="dyn_R_skirt_B_02"),Weight=0.5))
    Weight=0.5
   )
   > &RigUnit_ParentConstraint(
    AdvancedSettings=(RotationOrderForFilter=XZY)
    Child=(Type=Bone,Name="dyn_frontSkirt_03")
    Filter=(TranslationFilter=(bX=false,bY=false,bZ=false),RotationFilter=(bX=True,bY=True,bZ=True),ScaleFilter=(bX=True,bY=True,bZ=True))
    Parents=((Item=(Type=Bone,Name="dyn_L_skirt_B_03"),Weight=0.5),(Item=(Type=Bone,Name="dyn_R_skirt_B_03"),Weight=0.5))
    Weight=0.5
   )
   > &RigUnit_ParentConstraint(
    AdvancedSettings=(RotationOrderForFilter=XZY)
    Child=(Type=Bone,Name="dyn_frontSkirt_04")
    Filter=(TranslationFilter=(bX=false,bY=false,bZ=false),RotationFilter=(bX=True,bY=True,bZ=True),ScaleFilter=(bX=True,bY=True,bZ=True))
    Parents=((Item=(Type=Bone,Name="dyn_L_skirt_B_04"),Weight=0.5),(Item=(Type=Bone,Name="dyn_R_skirt_B_04"),Weight=0.5))
    Weight=0.5
   )
   > &RigUnit_ModifyTransforms(
    ItemToModify=((Item=(Type=Bone,Name="dyn_frontSkirt_master"),Transform=(Rotation=(X=0,Y=0,Z=0,W=1),Translation=(X=0,Y=0,Z=0),Scale3D=(X=1,Y=1,Z=1))))
    ItemToModify.0.Transform.Rotation=&RigVMFunction_MathQuaternionFromEuler(Euler=(X=0,Y=0,Z=30)).Result
    Weight=@n5.OutputParam
   )
   &RigUnit_RotationConstraintLocalSpaceOffset(
    AdvancedSettings=(RotationOrderForFilter=XZY)
    Child=(Type=Bone,Name="dyn_L_skirt_B_master")
    ExecutePin=()
    Filter=(bX=True,bY=True,bZ=True)
    Parents=((Item=(Type=Bone,Name="thigh_twist_01_l"),Weight=0.8),(Item=(Type=Bone,Name="pelvis"),Weight=0.2))
    Weight=1
   )
   > &RigUnit_RotationConstraintLocalSpaceOffset(
    AdvancedSettings=(RotationOrderForFilter=XZY)
    Child=(Type=Bone,Name="dyn_L_skirt_C_master")
    Filter=(bX=True,bY=True,bZ=True)
    Parents=((Item=(Type=Bone,Name="thigh_twist_01_l"),Weight=0.4),(Item=(Type=Bone,Name="pelvis"),Weight=0.6))
    Weight=1
   )
   n7 = &RigUnit_SphericalPoseReader(
    ActiveRegionScaleFactors=(PositiveWidth=2,NegativeWidth=0,PositiveHeight=1,NegativeHeight=1)
    ActiveRegionSize=0.1
    Debug=(bDrawDebug=True,bDraw2D=false,bDrawLocalAxes=false,DebugScale=25,DebugSegments=20,DebugThickness=0.25)
    DriverAxis=(X=-1,Y=0,Z=0)
    DriverItem=(Type=Bone,Name="thigh_l")
    ExecutePin=()
    FalloffRegionScaleFactors=(PositiveWidth=1,NegativeWidth=0,PositiveHeight=0.2,NegativeHeight=0.2)
    FalloffSize=0.45
    FlipHeightScaling=false
    FlipWidthScaling=true
    OptionalParentItem=(Type=Bone)
    RotationOffset=(X=110,Y=0,Z=-80)
   )
   n8 = &RigVMFunction_MathFloatMul(A=1, B=@n7.OutputParam) {template="Multiply::Execute(in A,in B,out Result)"}
   > &RigUnit_RotationConstraintLocalSpaceOffset(
    AdvancedSettings=(RotationOrderForFilter=XZY)
    Child=(Type=Bone,Name="dyn_L_skirt_B_02")
    Filter=(bX=True,bY=True,bZ=True)
    Parents=((Item=(Type=Bone,Name="thigh_twist_01_l"),Weight=0.8),(Item=(Type=Bone,Name="pelvis"),Weight=0.2))
    Weight=@n8.Result
   )
   > &RigUnit_ModifyTransforms(
    ItemToModify=((Item=(Type=Bone,Name="dyn_L_skirt_C_master"),Transform=(Rotation=(X=0,Y=0,Z=0,W=1),Translation=(X=0,Y=-15,Z=20),Scale3D=(X=1,Y=1,Z=1))))
    Weight=@n8.Result
   )
   > &RigUnit_ModifyTransforms(
    ItemToModify=((Item=(Type=Bone,Name="dyn_L_skirt_B_master"),Transform=(Rotation=(X=0,Y=0,Z=0,W=1),Translation=(X=0,Y=0,Z=10),Scale3D=(X=1,Y=1,Z=1))))
    Weight=@n8.Result
   )
   > &RigUnit_ModifyTransforms(
    ItemToModify=((Item=(Type=Bone,Name="dyn_L_skirt_B_02"),Transform=(Rotation=(X=0,Y=0,Z=0,W=1),Translation=(X=0,Y=0,Z=20),Scale3D=(X=1,Y=1,Z=1))))
    Weight=@n8.Result
   )
   > &RigUnit_ModifyTransforms(
    ItemToModify=((Item=(Type=Bone,Name="dyn_L_skirt_B_03"),Transform=(Rotation=(X=0,Y=0,Z=0,W=1),Translation=(X=0,Y=0,Z=-10),Scale3D=(X=1,Y=1,Z=1))))
    Weight=@n8.Result
   )
   > &RigUnit_ModifyTransforms(
    ItemToModify=((Item=(Type=Bone,Name="dyn_L_skirt_B_04"),Transform=(Rotation=(X=0,Y=0,Z=0,W=1),Translation=(X=0,Y=0,Z=-20),Scale3D=(X=1,Y=1,Z=1))))
    Weight=@n8.Result
   )
   n9 = &RigUnit_SphericalPoseReader(
    ActiveRegionScaleFactors=(PositiveWidth=3.5,NegativeWidth=0.5,PositiveHeight=3,NegativeHeight=3)
    ActiveRegionSize=0.1
    Debug=(bDrawDebug=True,DebugScale=25,DebugSegments=20,DebugThickness=0.25)
    DriverAxis=(X=-1,Y=0,Z=0)
    DriverItem=(Type=Bone,Name="thigh_l")
    ExecutePin=()
    FalloffRegionScaleFactors=(PositiveWidth=1,NegativeWidth=1,PositiveHeight=1,NegativeHeight=1)
    FalloffSize=0.45
    OptionalParentItem=(Type=Bone)
    RotationOffset=(X=90,Y=0,Z=0)
   )
   > &RigUnit_ModifyTransforms(
    ItemToModify=((Item=(Type=Bone,Name="dyn_L_skirt_B_master"),Transform=(Rotation=(X=0,Y=0,Z=0,W=1),Translation=(X=0,Y=0,Z=0),Scale3D=(X=1,Y=1,Z=1))))
    ItemToModify.0.Transform.Rotation=&RigVMFunction_MathQuaternionFromEuler(Euler=(X=0,Y=-3,Z=15)).Result
    Weight=@n9.OutputParam
   )
   > &RigUnit_ModifyTransforms(
    ItemToModify=((Item=(Type=Bone,Name="dyn_L_skirt_C_master"),Transform=(Rotation=(X=0,Y=0,Z=0,W=1),Translation=(X=0,Y=0,Z=0),Scale3D=(X=1,Y=1,Z=1))))
    ItemToModify.0.Transform.Rotation=&RigVMFunction_MathQuaternionFromEuler(Euler=(X=0,Y=-5,Z=0)).Result
    Weight=@n9.OutputParam
   )
   &RigVMFunction_MathFloatMul(B=@n5.OutputParam, A=@n3.OutputParam) {template="Multiply::Execute(in A,in B,out Result)"}
   &RigUnit_SphericalPoseReader(
    ActiveRegionScaleFactors=(PositiveWidth=1.5,NegativeWidth=1.5,PositiveHeight=0.7,NegativeHeight=0.9)
    ActiveRegionSize=0.11
    Debug=(bDrawDebug=True,DebugScale=25,DebugSegments=20,DebugThickness=0.25)
    DriverAxis=(X=-1,Y=0,Z=0)
    DriverItem=(Type=Bone,Name="thigh_l")
    ExecutePin=()
    FalloffRegionScaleFactors=(PositiveWidth=0.9,NegativeWidth=1,PositiveHeight=1.1,NegativeHeight=0.9)
    FalloffSize=0.45
    OptionalParentItem=(Type=Bone)
    RotationOffset=(X=105,Y=0,Z=10)
   )
  ]}
]
hierarchy [
 root {loc=<0 0 0>, rot=<0 0 0 1>}
 pelvis {parent=root, loc=<0 -0.5301 98.0571>, rot=<0.0127 0.707 -0.0127 -0.707>}
 spine_01 {parent=pelvis, loc=<6.1208 -0.9328 0>, rot=<0 0 0.0576 -0.9983>}
 spine_02 {parent=spine_01, loc=<6.6022 -0.0153 0>, rot=<0 0 0.0022 -1>}
 spine_03 {parent=spine_02, loc=<6.4919 -0.0105 0>, rot=<0 0 0.0015 -1>}
 spine_04 {parent=spine_03, loc=<6.9596 -0.007 0>, rot=<0 0 0.0009 -1>}
 spine_05 {parent=spine_04, loc=<6.9981 -0.0017 0>, rot=<0 0 -0.0042 -1>}
 neck_01 {parent=spine_05, loc=<11.3138 0.1529 0>, rot=<0 0 -0.2003 -0.9797>}
 neck_02 {parent=neck_01, loc=<4.5627 0 0>, rot=<0 0 0.0504 -0.9987>}
 head {parent=neck_02, loc=<5.3914 0 0>, rot=<0 0 0.115 -0.9934>}
 clavicle_l {parent=spine_05, loc=<5.4831 -0.1053 -3.2946>, rot=<0.0938 -0.7388 0.0307 -0.6666>}
 upperarm_l {parent=clavicle_l, loc=<9.9486 0 0>, rot=<0.0371 -0.2992 -0.0705 -0.9509>}
 lowerarm_l {parent=upperarm_l, loc=<26.4793 0 0>, rot=<0 0 -0.2382 -0.9712>}
 hand_l {parent=lowerarm_l, loc=<22.7458 0 0>, rot=<-0.6747 -0.0182 0.0605 -0.7354>}
 thumb_01_l {parent=hand_l, loc=<1.3118 -0.3191 -2.459>, rot=<0.6546 -0.4042 0.0664 -0.6354>}
 thumb_02_l {parent=thumb_01_l, loc=<5.9322 -0.0253 0.1416>, rot=<0.0121 0.0046 0.2964 -0.955>}
 thumb_03_l {parent=thumb_02_l, loc=<2.2223 0.3682 0>, rot=<0 0 0.0172 -0.9999>}
 index_metacarpal_l {parent=hand_l, loc=<1.6535 0.0183 -2.96>, rot=<0.0158 -0.0124 -0.0173 -0.9996>}
 index_01_l {parent=index_metacarpal_l, loc=<8.3371 0.1718 -0.2461>, rot=<0.0363 0.0499 0.142 -0.9879>}
 index_02_l {parent=index_01_l, loc=<3.4919 -0.3004 0>, rot=<0 0 0.2948 -0.9556>}
 index_03_l {parent=index_02_l, loc=<2.5578 -0.123 0>, rot=<0 0 0.1968 -0.9804>}
 middle_metacarpal_l {parent=hand_l, loc=<2.1856 -0.1725 -0.5984>, rot=<0.006 0.0168 -0.0329 -0.9993>}
 middle_01_l {parent=middle_metacarpal_l, loc=<7.6374 0.346 -0.1508>, rot=<-0.0085 0.0328 0.1813 -0.9829>}
 middle_02_l {parent=middle_01_l, loc=<3.9108 0 0>, rot=<0 0 0.2808 -0.9598>}
 middle_03_l {parent=middle_02_l, loc=<2.6103 0.0517 0>, rot=<0 0 0.1446 -0.9895>}
 ring_metacarpal_l {parent=hand_l, loc=<2.2074 -0.3625 1.3638>, rot=<-0.0284 0.0722 -0.0458 -0.9959>}
 ring_01_l {parent=ring_metacarpal_l, loc=<7.2483 0.0329 -0.1426>, rot=<-0.0456 -0.0277 0.2951 -0.954>}
 ring_02_l {parent=ring_01_l, loc=<3.7208 0.4647 0>, rot=<0 0 0.1583 -0.9874>}
 ring_03_l {parent=ring_02_l, loc=<2.4152 0.114 0>, rot=<0 0 0.1504 -0.9886>}
 pinky_metacarpal_l {parent=hand_l, loc=<1.7616 -0.9753 3.3814>, rot=<-0.0397 0.0814 -0.069 -0.9935>}
 pinky_01_l {parent=pinky_metacarpal_l, loc=<7.1933 -0.4205 -0.2601>, rot=<-0.0487 -0.0063 0.2556 -0.9655>}
 pinky_02_l {parent=pinky_01_l, loc=<2.8274 0.0893 0>, rot=<0 0 0.2474 -0.9689>}
 pinky_03_l {parent=pinky_02_l, loc=<2.2046 0.0307 0>, rot=<0 0 0.125 -0.9922>}
 weapon_l {parent=hand_l, loc=<-0.1974 -3.4111 0.3811>, rot=<0 0 0 1>}
 hand_attach_l {parent=hand_l, loc=<8.5 -3 0>, rot=<0.7071 0.7071 0 0>}
 lowerarm_twist_02_l {parent=lowerarm_l, loc=<9.6199 0 0>, rot=<0 0 0 -1>}
 lowerarm_twist_01_l {parent=lowerarm_l, loc=<17.8549 0 0>, rot=<0 0 0 -1>}
 upperarm_twist_01_l {parent=upperarm_l, loc=<0.0041 0 0>, rot=<0 0 0 1>}
 upperarm_twist_02_l {parent=upperarm_l, loc=<13.7941 0 0>, rot=<0 0 0 1>}
 clavicle_r {parent=spine_05, loc=<5.4831 -0.1053 3.2946>, rot=<0.7388 0.0938 -0.6666 -0.0307>}
 upperarm_r {parent=clavicle_r, loc=<-9.9486 0 0>, rot=<0.0371 -0.2992 -0.0705 -0.9509>}
 lowerarm_r {parent=upperarm_r, loc=<-26.4793 0 0>, rot=<0 0 -0.2382 -0.9712>}
 hand_r {parent=lowerarm_r, loc=<-22.7458 0 0>, rot=<-0.6747 -0.0182 0.0605 -0.7354>}
 thumb_01_r {parent=hand_r, loc=<-1.3118 0.3191 2.459>, rot=<0.6546 -0.4042 0.0664 -0.6354>}
 thumb_02_r {parent=thumb_01_r, loc=<-5.9322 0.0253 -0.1416>, rot=<0.0121 0.0046 0.2964 -0.955>}
 thumb_03_r {parent=thumb_02_r, loc=<-2.2223 -0.3682 0>, rot=<0 0 0.0172 -0.9999>}
 index_metacarpal_r {parent=hand_r, loc=<-1.6535 -0.0183 2.96>, rot=<0.0158 -0.0124 -0.0173 -0.9996>}
 index_01_r {parent=index_metacarpal_r, loc=<-8.3371 -0.1718 0.2461>, rot=<0.0363 0.0499 0.142 -0.9879>}
 index_02_r {parent=index_01_r, loc=<-3.4919 0.3004 0>, rot=<0 0 0.2948 -0.9556>}
 index_03_r {parent=index_02_r, loc=<-2.5578 0.123 0>, rot=<0 0 0.1968 -0.9804>}
 middle_metacarpal_r {parent=hand_r, loc=<-2.1856 0.1725 0.5984>, rot=<0.006 0.0168 -0.0329 -0.9993>}
 middle_01_r {parent=middle_metacarpal_r, loc=<-7.6374 -0.346 0.1508>, rot=<-0.0085 0.0328 0.1813 -0.9829>}
 middle_02_r {parent=middle_01_r, loc=<-3.9108 0 0>, rot=<0 0 0.2808 -0.9598>}
 middle_03_r {parent=middle_02_r, loc=<-2.6103 -0.0517 0>, rot=<0 0 0.1446 -0.9895>}
 ring_metacarpal_r {parent=hand_r, loc=<-2.2074 0.3625 -1.3638>, rot=<-0.0284 0.0722 -0.0458 -0.9959>}
 ring_01_r {parent=ring_metacarpal_r, loc=<-7.2483 -0.0329 0.1426>, rot=<-0.0456 -0.0277 0.2951 -0.954>}
 ring_02_r {parent=ring_01_r, loc=<-3.7208 -0.4647 0>, rot=<0 0 0.1583 -0.9874>}
 ring_03_r {parent=ring_02_r, loc=<-2.4152 -0.114 0>, rot=<0 0 0.1504 -0.9886>}
 pinky_metacarpal_r {parent=hand_r, loc=<-1.7616 0.9753 -3.3814>, rot=<-0.0397 0.0814 -0.069 -0.9935>}
 pinky_01_r {parent=pinky_metacarpal_r, loc=<-7.1933 0.4205 0.2601>, rot=<-0.0487 -0.0063 0.2556 -0.9655>}
 pinky_02_r {parent=pinky_01_r, loc=<-2.8274 -0.0893 0>, rot=<0 0 0.2474 -0.9689>}
 pinky_03_r {parent=pinky_02_r, loc=<-2.2046 -0.0307 0>, rot=<0 0 0.125 -0.9922>}
 weapon_r {parent=hand_r, loc=<0.1974 3.4111 -0.3811>, rot=<0 0 0 1>}
 hand_attach_r {parent=hand_r, loc=<-8.5001 2.9999 0>, rot=<0 0 -0.7071 -0.7071>}
 lowerarm_twist_02_r {parent=lowerarm_r, loc=<-9.6199 0 0>, rot=<0 0 0 -1>}
 lowerarm_twist_01_r {parent=lowerarm_r, loc=<-17.8549 0 0>, rot=<0 0 0 1>}
 upperarm_twist_01_r {parent=upperarm_r, loc=<-0.0041 0 0>, rot=<0 0 0 -1>}
 upperarm_twist_02_r {parent=upperarm_r, loc=<-13.7941 0 0>, rot=<0 0 0 -1>}
 dyn_frontSkirt_master {parent=spine_02, loc=<3.5441 8.1605 0.0057>, rot=<-0.1917 -0.9814 0.0026 0.0091>}
 dyn_frontSkirt_01 {parent=dyn_frontSkirt_master, loc=<0 0 0>, rot=<0.7071 0 0 -0.7071>}
 dyn_frontSkirt_02 {parent=dyn_frontSkirt_01, loc=<9.5963 0 0>, rot=<0.0038 -0.0587 0.0087 -0.9982>}
 dyn_frontSkirt_03 {parent=dyn_frontSkirt_02, loc=<10.9877 0 0>, rot=<0.0011 -0.0436 0.0034 -0.999>}
 dyn_frontSkirt_04 {parent=dyn_frontSkirt_03, loc=<11.6445 0 0>, rot=<0 0 0 1>}
 dyn_L_skirt_B_master {parent=spine_02, loc=<3.6999 6.3076 -6.6168>, rot=<-0.0815 -0.846 -0.4681 -0.2421>}
 dyn_L_skirt_B_01 {parent=dyn_L_skirt_B_master, loc=<0 0 0>, rot=<0 0 0 1>}
 dyn_L_skirt_B_02 {parent=dyn_L_skirt_B_01, loc=<10.2352 0 0>, rot=<-0.0192 -0.0814 0.0033 -0.9965>}
 dyn_L_skirt_B_03 {parent=dyn_L_skirt_B_02, loc=<11.287 0 0>, rot=<-0.0121 -0.0767 0.0015 -0.997>}
 dyn_L_skirt_B_04 {parent=dyn_L_skirt_B_03, loc=<11.6446 0 0>, rot=<0 0 0 1>}
 dyn_L_skirt_C_master {parent=spine_02, loc=<4.1493 0.9638 -8.5725>, rot=<0.0419 -0.9572 -0.0115 -0.2862>}
 dyn_L_skirt_C_01 {parent=dyn_L_skirt_C_master, loc=<0 0 0>, rot=<0 0 0 1>}
 dyn_L_skirt_C_02 {parent=dyn_L_skirt_C_01, loc=<10.2355 0 0>, rot=<-0.0004 -0.081 -0.0084 -0.9967>}
 dyn_L_skirt_C_03 {parent=dyn_L_skirt_C_02, loc=<11.2867 0 0>, rot=<-0.0013 -0.0766 -0.0038 -0.997>}
 dyn_L_skirt_C_04 {parent=dyn_L_skirt_C_03, loc=<11.6446 0 0>, rot=<0 0 0 1>}
 dyn_L_skirt_D_master {parent=spine_02, loc=<4.5019 -3.2295 -6.0638>, rot=<0.1689 -0.8619 0.3943 -0.2706>}
 dyn_L_skirt_D_01 {parent=dyn_L_skirt_D_master, loc=<0 0 0>, rot=<0 0 0 1>}
 dyn_L_skirt_D_02 {parent=dyn_L_skirt_D_01, loc=<10.2355 0 0>, rot=<0.0192 -0.0804 -0.0132 -0.9965>}
 dyn_L_skirt_D_03 {parent=dyn_L_skirt_D_02, loc=<11.2867 0 0>, rot=<0.0121 -0.076 -0.0109 -0.997>}
 dyn_L_skirt_D_04 {parent=dyn_L_skirt_D_03, loc=<11.6446 0 0>, rot=<0 0 0 1>}
 dyn_R_skirt_B_master {parent=spine_02, loc=<3.6999 6.3076 6.6168>, rot=<0.846 -0.0815 -0.2421 0.4681>}
 dyn_R_skirt_B_01 {parent=dyn_R_skirt_B_master, loc=<0 0 0>, rot=<0 0 0 1>}
 dyn_R_skirt_B_02 {parent=dyn_R_skirt_B_01, loc=<-10.2353 0 0>, rot=<-0.0192 -0.0814 0.0033 -0.9965>}
 dyn_R_skirt_B_03 {parent=dyn_R_skirt_B_02, loc=<-11.287 0 0>, rot=<-0.0121 -0.0767 0.0015 -0.997>}
 dyn_R_skirt_B_04 {parent=dyn_R_skirt_B_03, loc=<-11.6446 0 0>, rot=<0 0 0 1>}
 dyn_R_skirt_C_master {parent=spine_02, loc=<4.1493 0.9638 8.5725>, rot=<0.9572 0.0419 -0.2862 0.0115>}
 dyn_R_skirt_C_01 {parent=dyn_R_skirt_C_master, loc=<0 0 0>, rot=<0 0 0 1>}
 dyn_R_skirt_C_02 {parent=dyn_R_skirt_C_01, loc=<-10.2352 0 -0.0001>, rot=<-0.0004 -0.081 -0.0084 -0.9967>}
 dyn_R_skirt_C_03 {parent=dyn_R_skirt_C_02, loc=<-11.287 0 0.0001>, rot=<-0.0013 -0.0766 -0.0038 -0.997>}
 dyn_R_skirt_C_04 {parent=dyn_R_skirt_C_03, loc=<-11.6446 0 0>, rot=<0 0 0 1>}
 dyn_R_skirt_D_master {parent=spine_02, loc=<4.5019 -3.2295 6.0638>, rot=<0.8619 0.1689 -0.2706 -0.3943>}
 dyn_R_skirt_D_01 {parent=dyn_R_skirt_D_master, loc=<0 0 0>, rot=<0 0 0 -1>}
 dyn_R_skirt_D_02 {parent=dyn_R_skirt_D_01, loc=<-10.2352 0 -0.0002>, rot=<0.0192 -0.0804 -0.0132 -0.9965>}
 dyn_R_skirt_D_03 {parent=dyn_R_skirt_D_02, loc=<-11.287 0 0.0002>, rot=<0.0121 -0.076 -0.0109 -0.997>}
 dyn_R_skirt_D_04 {parent=dyn_R_skirt_D_03, loc=<-11.6446 0 0>, rot=<0 0 0 1>}
 dyn_backSkirt_master {parent=spine_02, loc=<4.7008 -5.5952 0.0057>, rot=<-0.2242 0.6668 0.67 -0.2371>}
 dyn_backSkirt_01 {parent=dyn_backSkirt_master, loc=<0 0 0>, rot=<1 0 0 0>}
 dyn_backSkirt_02 {parent=dyn_backSkirt_01, loc=<10.2355 0 0>, rot=<-0.0045 -0.081 -0.0083 -0.9967>}
 dyn_backSkirt_03 {parent=dyn_backSkirt_02, loc=<11.2867 0 0>, rot=<-0.0012 -0.0767 -0.0034 -0.9971>}
 dyn_backSkirt_04 {parent=dyn_backSkirt_03, loc=<11.6446 0 0>, rot=<0 0 0 1>}
 thigh_l {parent=pelvis, loc=<-1.0855 -0.2109 -8.7053>, rot=<0.0009 0.0345 0.0155 -0.9993>}
 calf_l {parent=thigh_l, loc=<-42.5268 0 0>, rot=<0 0 -0.0657 -0.9978>}
 foot_l {parent=calf_l, loc=<-46.8037 0 0>, rot=<0.0422 -0.04 0.0613 -0.9964>}
 ball_l {parent=foot_l, loc=<-3.5658 10.8066 -0.1544>, rot=<0.0001 -0.0001 -0.7063 -0.7079>}
 calf_twist_02_l {parent=calf_l, loc=<-18.8064 -0.0967 0.0397>, rot=<0.0237 0.0095 -0.0035 -0.9997>}
 calf_twist_01_l {parent=calf_l, loc=<-35.813 -0.0967 0.0397>, rot=<0.0237 0.0095 -0.0035 -0.9997>}
 thigh_twist_01_l {parent=thigh_l, loc=<-21.629 -0.0919 -0.0002>, rot=<0.0013 0 0 -1>}
 thigh_r {parent=pelvis, loc=<-1.0855 -0.2109 8.7053>, rot=<0.0345 -0.0009 0.9993 0.0155>}
 calf_r {parent=thigh_r, loc=<42.5268 0 0>, rot=<0 0 -0.0657 -0.9978>}
 foot_r {parent=calf_r, loc=<46.8037 0 0>, rot=<0.0422 -0.04 0.0613 -0.9964>}
 ball_r {parent=foot_r, loc=<3.5658 -10.8066 0.1544>, rot=<0.0001 -0.0001 -0.7063 -0.7079>}
 calf_twist_02_r {parent=calf_r, loc=<18.8064 0.0967 -0.0397>, rot=<0.0237 0.0095 -0.0035 -0.9997>}
 calf_twist_01_r {parent=calf_r, loc=<35.813 0.0967 -0.0397>, rot=<0.0237 0.0095 -0.0035 -0.9997>}
 thigh_twist_01_r {parent=thigh_r, loc=<21.629 0.0919 0.0002>, rot=<0.0013 0 0 -1>}
 attach {parent=root, loc=<0 0 0>, rot=<0 0 0 1>}
]
variables [ModifyTransforms_1_1_ItemToModify_0_Transform {type=FTransform, default=(Rotation=(X=0,Y=0,Z=0,W=1),Translation=(X=0,Y=0,Z=0),Scale3D=(X=1,Y=1,Z=1))}]
