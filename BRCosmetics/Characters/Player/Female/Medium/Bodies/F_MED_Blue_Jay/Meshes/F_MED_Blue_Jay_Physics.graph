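# Physics asset for the F_MED_Blue_Jay female player body, set up for skirt physics around the pelvis and thighs.
# Six skirt chains (B, C, D per side) hang from master bodies with free-angular pelvis joints, using tight twist and limited swing down each chain, stiffest in B.
# Free-angular cross-links tie neighboring hem bodies on each side together.
# Thigh capsules are angularly locked to the pelvis; skirt bodies are capsules except boxes at B tips.
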PhysicsAsset /BRCosmetics/Characters/Player/Female/Medium/Bodies/F_MED_Blue_Jay/Meshes/F_MED_Blue_Jay_Physics
bodies {
 dyn_L_skirt_B_02 [{shapeType=Capsule, length=7, radius=4.5, shapeName=dyn_L_skirt_D_03_capsule, center=<2 0 0>, rotation=<rot 90 0 0>}]
 dyn_L_skirt_B_03 [{shapeType=Capsule, length=7.375, radius=4.125, shapeName=dyn_L_skirt_D_03_capsule, center=<2 0 0>, rotation=<rot 90 0 0>}]
 dyn_L_skirt_B_04 [{shapeType=Box, extentX=15, extentY=27.5, extentZ=8.25, shapeName=dyn_L_skirt_B_04_box, center=<1.514 0 0>, rotation=<rot -6.6568 0 7.1813>}]
 dyn_L_skirt_B_master [{shapeType=Capsule, length=0.5, radius=0.5, shapeName=dyn_L_skirt_B_master_capsule}]
 dyn_L_skirt_C_02 [{shapeType=Capsule, length=4, radius=3, shapeName=dyn_L_skirt_C_02_capsule, center=<3.4589 0 2.5>, rotation=<rot 90 0 0>}]
 dyn_L_skirt_C_03 [{shapeType=Capsule, length=7, radius=4.5, shapeName=dyn_L_skirt_D_03_capsule, center=<2 0 2.9>, rotation=<rot 90 0 0>}]
 dyn_L_skirt_C_04 [{shapeType=Capsule, length=7.375, radius=6.625, shapeName=dyn_L_skirt_D_03_capsule, center=<2 0 3.3>, rotation=<rot 90 0 0>}]
 dyn_L_skirt_C_master [{shapeType=Capsule, length=0.5, radius=0.5, shapeName=dyn_L_skirt_C_master_capsule}]
 dyn_L_skirt_D_02 [{shapeType=Capsule, length=4, radius=3.8, shapeName=dyn_L_skirt_D_02_capsule, center=<3.4589 0 0>, rotation=<rot 90 0 0>}]
 dyn_L_skirt_D_03 [{shapeType=Capsule, length=7, radius=4.5, shapeName=dyn_L_skirt_D_03_capsule, center=<2 0 0>, rotation=<rot 90 0 0>}]
 dyn_L_skirt_D_04 [{shapeType=Capsule, length=7, radius=7, shapeName=dyn_L_skirt_D_03_capsule, center=<2 0 0>, rotation=<rot 90 0 0>}]
 dyn_L_skirt_D_master [{shapeType=Capsule, length=0.5, radius=0.5, shapeName=dyn_L_skirt_D_master_capsule}]
 dyn_R_skirt_B_02 [{shapeType=Capsule, length=7, radius=4.5, shapeName=dyn_R_skirt_D_03_capsule, center=<-2 0 0>, rotation=<rot 90 0 0>}]
 dyn_R_skirt_B_03 [{shapeType=Capsule, length=7, radius=4.5, shapeName=dyn_R_skirt_D_03_capsule, center=<-2 0 0>, rotation=<rot 90 0 0>}]
 dyn_R_skirt_B_04 [{shapeType=Box, extentX=15, extentY=27.5, extentZ=8.25, shapeName=dyn_R_skirt_B_04_box, center=<-1.514 0 0>, rotation=<rot 0 0 7.1813>}]
 dyn_R_skirt_B_master [{shapeType=Capsule, length=0.5, radius=0.5, shapeName=dyn_R_skirt_B_master_capsule}]
 dyn_R_skirt_C_02 [{shapeType=Capsule, length=4, radius=3, shapeName=dyn_R_skirt_C_02_capsule, center=<-3.4589 0 -2.5>, rotation=<rot 90 0 0>}]
 dyn_R_skirt_C_03 [{shapeType=Capsule, length=7, radius=4.5, shapeName=dyn_R_skirt_D_03_capsule, center=<-2 0 -2.9>, rotation=<rot 90 0 0>}]
 dyn_R_skirt_C_04 [{shapeType=Capsule, length=7, radius=7, shapeName=dyn_L_skirt_D_03_capsule, center=<-2 0 -3.3>, rotation=<rot 90 0 0>}]
 dyn_R_skirt_C_master [{shapeType=Capsule, length=0.5, radius=0.5, shapeName=dyn_R_skirt_C_master_capsule}]
 dyn_R_skirt_D_02 [{shapeType=Capsule, length=4, radius=3.8, shapeName=dyn_R_skirt_D_02_capsule, center=<-3.4589 0 0>, rotation=<rot 90 0 0>}]
 dyn_R_skirt_D_03 [{shapeType=Capsule, length=7, radius=4.5, shapeName=dyn_R_skirt_D_03_capsule, center=<-2 0 0>, rotation=<rot 90 0 0>}]
 dyn_R_skirt_D_04 [{shapeType=Capsule, length=7, radius=7, shapeName=dyn_R_skirt_D_03_capsule, center=<-2 0 0>, rotation=<rot 90 0 0>}]
 dyn_R_skirt_D_master [{shapeType=Capsule, length=0.5, radius=0.5, shapeName=dyn_R_skirt_D_master_capsule}]
 pelvis [{shapeType=Capsule, length=12.375, radius=8.5, shapeName=pelvis_capsule}]
 thigh_l [{shapeType=Capsule, length=40, radius=8, shapeName=thigh_l_capsule, center=<-20 0 0>, rotation=<rot 90 0 0>}]
 thigh_r [{shapeType=Capsule, length=40, radius=8, shapeName=thigh_l_capsule, center=<20 0 0>, rotation=<rot 90 0 0>}]
}
constraints [
 {bone1Name=dyn_L_skirt_B_02, bone2Name=dyn_L_skirt_B_master, swing1LimitDegrees=3, swing2LimitDegrees=20, twistLimitDegrees=1}
 {bone1Name=dyn_L_skirt_B_03, bone2Name=dyn_L_skirt_B_02, swing1LimitDegrees=3, swing2LimitDegrees=20, twistLimitDegrees=1}
 {bone1Name=dyn_L_skirt_B_04, bone2Name=dyn_L_skirt_B_03, swing1LimitDegrees=3, swing2LimitDegrees=20, twistLimitDegrees=1}
 {bone1Name=dyn_L_skirt_C_02, bone2Name=dyn_L_skirt_C_master, swing1LimitDegrees=20, swing2LimitDegrees=20, twistLimitDegrees=1}
 {bone1Name=dyn_L_skirt_C_03, bone2Name=dyn_L_skirt_C_02, swing1LimitDegrees=20, swing2LimitDegrees=20, twistLimitDegrees=1}
 {bone1Name=dyn_L_skirt_C_04, bone2Name=dyn_L_skirt_C_03, swing1LimitDegrees=20, swing2LimitDegrees=20, twistLimitDegrees=1}
 {bone1Name=dyn_L_skirt_D_02, bone2Name=dyn_L_skirt_D_master, swing1LimitDegrees=20, swing2LimitDegrees=20, twistLimitDegrees=1}
 {bone1Name=dyn_L_skirt_D_03, bone2Name=dyn_L_skirt_D_02, swing1LimitDegrees=20, swing2LimitDegrees=20, twistLimitDegrees=1}
 {bone1Name=dyn_L_skirt_D_04, bone2Name=dyn_L_skirt_D_03, swing1LimitDegrees=20, swing2LimitDegrees=20, twistLimitDegrees=1}
 {bone1Name=dyn_R_skirt_B_02, bone2Name=dyn_R_skirt_B_master, swing1LimitDegrees=3, swing2LimitDegrees=20, twistLimitDegrees=1}
 {bone1Name=dyn_R_skirt_B_03, bone2Name=dyn_R_skirt_B_02, swing1LimitDegrees=3, swing2LimitDegrees=20, twistLimitDegrees=1}
 {bone1Name=dyn_R_skirt_B_04, bone2Name=dyn_R_skirt_B_03, swing1LimitDegrees=3, swing2LimitDegrees=20, twistLimitDegrees=1}
 {bone1Name=dyn_R_skirt_C_02, bone2Name=dyn_R_skirt_C_master, swing1LimitDegrees=20, swing2LimitDegrees=20, twistLimitDegrees=1}
 {bone1Name=dyn_R_skirt_C_03, bone2Name=dyn_R_skirt_C_02, swing1LimitDegrees=20, swing2LimitDegrees=20, twistLimitDegrees=1}
 {bone1Name=dyn_R_skirt_C_04, bone2Name=dyn_R_skirt_C_03, swing1LimitDegrees=20, swing2LimitDegrees=20, twistLimitDegrees=1}
 {bone1Name=dyn_R_skirt_D_02, bone2Name=dyn_R_skirt_D_master, swing1LimitDegrees=20, swing2LimitDegrees=20, twistLimitDegrees=1}
 {bone1Name=dyn_R_skirt_D_03, bone2Name=dyn_R_skirt_D_02, swing1LimitDegrees=20, swing2LimitDegrees=20, twistLimitDegrees=1}
 {bone1Name=dyn_R_skirt_D_04, bone2Name=dyn_R_skirt_D_03, swing1LimitDegrees=20, swing2LimitDegrees=20, twistLimitDegrees=1}
 {
  bone1Name=thigh_l
  bone2Name=pelvis
  swing1LimitDegrees=45
  swing1Motion=Locked
  swing2LimitDegrees=45
  swing2Motion=Locked
  twistLimitDegrees=45
  twistMotion=Locked
 }
 {
  bone1Name=thigh_r
  bone2Name=pelvis
  swing1LimitDegrees=45
  swing1Motion=Locked
  swing2LimitDegrees=45
  swing2Motion=Locked
  twistLimitDegrees=45
  twistMotion=Locked
 }
 {
  bone1Name=dyn_L_skirt_B_master
  bone2Name=pelvis
  swing1LimitDegrees=45
  swing1Motion=Free
  swing2LimitDegrees=45
  swing2Motion=Free
  twistLimitDegrees=45
  twistMotion=Free
 }
 {
  bone1Name=dyn_L_skirt_C_master
  bone2Name=pelvis
  swing1LimitDegrees=45
  swing1Motion=Free
  swing2LimitDegrees=45
  swing2Motion=Free
  twistLimitDegrees=45
  twistMotion=Free
 }
 {
  bone1Name=dyn_L_skirt_D_master
  bone2Name=pelvis
  swing1LimitDegrees=45
  swing1Motion=Free
  swing2LimitDegrees=45
  swing2Motion=Free
  twistLimitDegrees=45
  twistMotion=Free
 }
 {
  bone1Name=dyn_R_skirt_B_master
  bone2Name=pelvis
  swing1LimitDegrees=45
  swing1Motion=Free
  swing2LimitDegrees=45
  swing2Motion=Free
  twistLimitDegrees=45
  twistMotion=Free
 }
 {
  bone1Name=dyn_R_skirt_C_master
  bone2Name=pelvis
  swing1LimitDegrees=45
  swing1Motion=Free
  swing2LimitDegrees=45
  swing2Motion=Free
  twistLimitDegrees=45
  twistMotion=Free
 }
 {
  bone1Name=dyn_R_skirt_D_master
  bone2Name=pelvis
  swing1LimitDegrees=45
  swing1Motion=Free
  swing2LimitDegrees=45
  swing2Motion=Free
  twistLimitDegrees=45
  twistMotion=Free
 }
 {
  bone1Name=dyn_R_skirt_C_04
  bone2Name=dyn_R_skirt_B_04
  swing1LimitDegrees=45
  swing1Motion=Free
  swing2LimitDegrees=45
  swing2Motion=Free
  twistLimitDegrees=45
  twistMotion=Free
 }
 {
  bone1Name=dyn_L_skirt_C_04
  bone2Name=dyn_L_skirt_B_04
  swing1LimitDegrees=45
  swing1Motion=Free
  swing2LimitDegrees=45
  swing2Motion=Free
  twistLimitDegrees=45
  twistMotion=Free
 }
 {
  bone1Name=dyn_R_skirt_D_04
  bone2Name=dyn_R_skirt_C_04
  swing1LimitDegrees=45
  swing1Motion=Free
  swing2LimitDegrees=45
  swing2Motion=Free
  twistLimitDegrees=45
  twistMotion=Free
 }
 {
  bone1Name=dyn_L_skirt_D_04
  bone2Name=dyn_L_skirt_C_04
  swing1LimitDegrees=45
  swing1Motion=Free
  swing2LimitDegrees=45
  swing2Motion=Free
  twistLimitDegrees=45
  twistMotion=Free
 }
]
