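# Animation Blueprint for the F_MED_Blue_Jay female body cosmetic, adding skirt dynamics and a backpack offset to the copied parent mesh pose.
# Spine-weighted constraints and a Control Rig feed cached 'presim', then downed-state offsets, pelvis-space RigidBody simulation and constraints averaging neighboring skirt chains.
# The event graph sets downed, backpack-tag, movement and thigh-angle values.
# Movement and thigh-driven skirt nodes are not wired to the output pose.
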AnimBlueprint /BRCosmetics/Characters/Player/Female/Medium/Bodies/F_MED_Blue_Jay/Meshes/F_MED_Blue_Jay_AnimBP
target_skeleton=&/F_MED_Blue_Jay_Skeleton
graphs [
 AnimGraph {nodes [
   n1 = Get Movement() -> (Movement:bool)
   n2 = "Constraint - dyn_L_skirt_B_master (RR)"!(ComponentPose, bAlphaBoolEnabled:bool=@n1, ConstraintWeights_0=0.2, ConstraintWeights_1:float=0.8) -> (Pose) {
    alphaInputType=Bool
    alphaBoolBlend {blendInTime=0.1, blendOption=Linear, blendOutTime=0.1}
    boneToModify {boneName=dyn_L_skirt_B_master}
    constraintSetup [{transformType=Rotation, targetBone {boneName=pelvis}}, {transformType=Rotation, targetBone {boneName=thigh_l}}]
    constraintWeights [1,1]
   }
   n3 = "Constraint - dyn_R_skirt_B_master (RR)"!(ComponentPose=@n2, bAlphaBoolEnabled:bool=@n1, ConstraintWeights_0=0.2, ConstraintWeights_1:float=0.8) -> (Pose) {
    alphaInputType=Bool
    alphaBoolBlend {blendInTime=0.1, blendOption=Linear, blendOutTime=0.1}
    boneToModify {boneName=dyn_R_skirt_B_master}
    constraintSetup [{transformType=Rotation, targetBone {boneName=pelvis}}, {transformType=Rotation, targetBone {boneName=thigh_r}}]
    constraintWeights [1,1]
   }
   n4 = "Constraint - dyn_L_skirt_C_master (RR)"!(ComponentPose=@n3, bAlphaBoolEnabled:bool=@n1, ConstraintWeights_0=0.6, ConstraintWeights_1:float=0.4) -> (Pose) {
    alphaInputType=Bool
    alphaBoolBlend {blendInTime=0.1, blendOption=Linear, blendOutTime=0.1}
    boneToModify {boneName=dyn_L_skirt_C_master}
    constraintSetup [{transformType=Rotation, targetBone {boneName=pelvis}}, {transformType=Rotation, targetBone {boneName=thigh_l}}]
    constraintWeights [1,1]
   }
   "Constraint - dyn_R_skirt_C_master (RR)"!(ComponentPose=@n4, bAlphaBoolEnabled:bool=@n1, ConstraintWeights_0=0.6, ConstraintWeights_1:float=0.4) -> (Pose) {
    alphaInputType=Bool
    alphaBoolBlend {blendInTime=0.1, blendOption=Linear, blendOutTime=0.1}
    boneToModify {boneName=dyn_R_skirt_C_master}
    constraintSetup [{transformType=Rotation, targetBone {boneName=pelvis}}, {transformType=Rotation, targetBone {boneName=thigh_r}}]
    constraintWeights [1,1]
   }
   Local To Component()
   n5 = Get L_thighRotationAlpha() -> (L_thighRotationAlpha:double)
   n6 = "Transform (Modify) Bone - Bone: dyn_L_skirt_B_master"(Alpha=@n5, Translation="0,0,0", Rotation="-4.757,8.640,11.119", Scale="1,1,1") {
    rotationMode=BMM_Additive
    rotationSpace=BCS_BoneSpace
    translationMode=BMM_Additive
    translationSpace=BCS_BoneSpace
    alphaBoolBlend {blendOption=Linear}
    boneToModify {boneName=dyn_L_skirt_B_master}
    rotation=<rot -4.757 8.64 11.119>
   }
   n7 = Get R_thighRotationAlpha() -> (R_thighRotationAlpha:double)
   "Transform (Modify) Bone - Bone: dyn_R_skirt_B_master"(ComponentPose=@n6, Alpha=@n7, Translation="0,0,0", Rotation="0,16,-4", Scale="1,1,1") {
    rotationMode=BMM_Additive
    rotationSpace=BCS_BoneSpace
    translationMode=BMM_Additive
    translationSpace=BCS_BoneSpace
    alphaBoolBlend {blendOption=Linear}
    boneToModify {boneName=dyn_R_skirt_B_master}
    rotation {roll=-4, yaw=16}
   }
   "Transform (Modify) Bone - Bone: dyn_R_skirt_D_master"(Alpha=@n7, Translation="0,0,0", Rotation="0,36,0", Scale="1,1,1") {
    rotationMode=BMM_Additive
    rotationSpace=BCS_BoneSpace
    translationMode=BMM_Additive
    translationSpace=BCS_BoneSpace
    alphaBoolBlend {blendOption=Linear}
    boneToModify {boneName=dyn_R_skirt_D_master}
    rotation {yaw=36}
   }
   n8 = Get L_thighRotationAlpha()
   n9 = "Transform (Modify) Bone - Bone: dyn_R_skirt_B_master"(Alpha=@n8, Translation="0,0,0", Rotation="15,20,16", Scale="1,1,1") {
    rotationMode=BMM_Additive
    rotationSpace=BCS_BoneSpace
    translationMode=BMM_Additive
    translationSpace=BCS_BoneSpace
    alphaBoolBlend {blendOption=Linear}
    alphaScaleBiasClamp {bInterpResult=true}
    boneToModify {boneName=dyn_R_skirt_B_master}
    rotation=<rot 15 20 16>
   }
   "Transform (Modify) Bone - Bone: dyn_frontSkirt_master"(ComponentPose=@n9, Alpha=@n8, Translation="0,0,0", Rotation="20,10,10", Scale="1,1,1") {
    rotationMode=BMM_Additive
    rotationSpace=BCS_BoneSpace
    translationMode=BMM_Additive
    translationSpace=BCS_BoneSpace
    alphaBoolBlend {blendOption=Linear}
    alphaScaleBiasClamp {bInterpResult=true}
    boneToModify {boneName=dyn_frontSkirt_master}
    rotation=<rot 20 10 10>
   }
   "Transform (Modify) Bone - Bone: dyn_L_skirt_D_master"(Alpha=@n5, Translation="0,0,0", Rotation="-9,42,-9", Scale="1,1,1") {
    rotationMode=BMM_Additive
    rotationSpace=BCS_BoneSpace
    translationMode=BMM_Additive
    translationSpace=BCS_BoneSpace
    alphaBoolBlend {blendOption=Linear}
    boneToModify {boneName=dyn_L_skirt_D_master}
    rotation=<rot -9 42 -9>
   }
   n10 = Get isDBNO() -> (isDBNO:double)
   n11 = "Transform (Modify) Bone - Bone: dyn_L_skirt_D_master"(
    ComponentPose=Local To Component(LocalPose=Use cached pose 'presim'())
    Alpha=@n10
    Translation="0,0,0"
    Rotation="-10,-7.5,-2"
    Scale="1,1,1"
   ) {
    rotationMode=BMM_Additive
    rotationSpace=BCS_BoneSpace
    translationMode=BMM_Additive
    translationSpace=BCS_BoneSpace
    alphaBoolBlend {blendOption=Linear}
    alphaScaleBiasClamp {bInterpResult=true}
    boneToModify {boneName=dyn_L_skirt_D_master}
    rotation=<rot -10 -7.5 -2>
   }
   n12 = "Transform (Modify) Bone - Bone: dyn_L_skirt_D_02"(ComponentPose=@n11, Alpha=@n10, Translation="0,0,0", Rotation="-10,0,0", Scale="1,1,1") {
    rotationMode=BMM_Additive
    rotationSpace=BCS_BoneSpace
    translationMode=BMM_Additive
    translationSpace=BCS_BoneSpace
    alphaBoolBlend {blendOption=Linear}
    alphaScaleBiasClamp {bInterpResult=true}
    boneToModify {boneName=dyn_L_skirt_D_02}
    rotation {pitch=-10}
   }
   n13 = "Transform (Modify) Bone - Bone: dyn_R_skirt_D_master"(ComponentPose=@n12, Alpha=@n10, Translation="0,0,0", Rotation="-10,-10,0", Scale="1,1,1") {
    rotationMode=BMM_Additive
    rotationSpace=BCS_BoneSpace
    translationMode=BMM_Additive
    translationSpace=BCS_BoneSpace
    alphaBoolBlend {blendOption=Linear}
    alphaScaleBiasClamp {bInterpResult=true}
    boneToModify {boneName=dyn_R_skirt_D_master}
    rotation {pitch=-10, yaw=-10}
   }
   n14 = "Transform (Modify) Bone - Bone: dyn_backSkirt_master"(ComponentPose=@n13, Alpha=@n10, Translation="0,0,0", Rotation="-10,0,-10", Scale="1,1,1") {
    rotationMode=BMM_Additive
    rotationSpace=BCS_BoneSpace
    translationMode=BMM_Additive
    translationSpace=BCS_BoneSpace
    alphaBoolBlend {blendOption=Linear}
    alphaScaleBiasClamp {bInterpResult=true}
    boneToModify {boneName=dyn_backSkirt_master}
    rotation {pitch=-10, roll=-10}
   }
   n15 = RigidBody(ComponentPose=@n14) {
    bClampLinearTranslationLimitToRefPose=true
    bOverrideWorldGravity=true
    overridePhysicsAsset=&/F_MED_Blue_Jay_Physics
    simulationSpace=BaseBoneSpace
    alphaBoolBlend {blendOption=Linear}
    baseBoneRef {boneName=pelvis}
    componentAppliedLinearAccClamp=<3000 3000 3000>
    componentLinearAccScale=<0.3 0.3 0.3>
    componentLinearVelScale=<0.3 0.3 0.3>
    overrideWorldGravity {z=-980}
    simSpaceSettings {maxAngularAcceleration=100, maxAngularVelocity=9, maxLinearAcceleration=300, maxLinearVelocity=400, velocityScaleZ=0.12, worldAlpha=1}
   }
   n16 = "Constraint - dyn_backSkirt_02 (RR)"!(ComponentPose=@n15, Alpha=1, ConstraintWeights_0=0.5, ConstraintWeights_1:float=0.5) -> (Pose) {
    alphaBoolBlend {blendOption=Linear}
    boneToModify {boneName=dyn_backSkirt_02}
    constraintSetup [{transformType=Rotation, targetBone {boneName=dyn_L_skirt_D_02}}, {transformType=Rotation, targetBone {boneName=dyn_R_skirt_D_02}}]
    constraintWeights [1,1]
   }
   n17 = "Constraint - dyn_backSkirt_03 (RR)"!(ComponentPose=@n16, Alpha=1, ConstraintWeights_0=0.5, ConstraintWeights_1:float=0.5) -> (Pose) {
    alphaBoolBlend {blendOption=Linear}
    boneToModify {boneName=dyn_backSkirt_03}
    constraintSetup [{transformType=Rotation, targetBone {boneName=dyn_L_skirt_D_03}}, {transformType=Rotation, targetBone {boneName=dyn_R_skirt_D_03}}]
    constraintWeights [1,1]
   }
   n18 = "Constraint - dyn_backSkirt_04 (RR)"!(ComponentPose=@n17, Alpha=1, ConstraintWeights_0=0.5, ConstraintWeights_1:float=0.5) -> (Pose) {
    alphaBoolBlend {blendOption=Linear}
    boneToModify {boneName=dyn_backSkirt_04}
    constraintSetup [{transformType=Rotation, targetBone {boneName=dyn_L_skirt_D_04}}, {transformType=Rotation, targetBone {boneName=dyn_R_skirt_D_04}}]
    constraintWeights [1,1]
   }
   n19 = "Constraint - dyn_frontSkirt_master (PP)"!(ComponentPose=@n18, Alpha=1, ConstraintWeights_0=0.5, ConstraintWeights_1:float=0.5) -> (Pose) {
    alphaBoolBlend {blendOption=Linear}
    boneToModify {boneName=dyn_frontSkirt_master}
    constraintSetup [{targetBone {boneName=dyn_L_skirt_B_master}}, {targetBone {boneName=dyn_R_skirt_B_master}}]
    constraintWeights [1,1]
   }
   n20 = "Constraint - dyn_frontSkirt_01 (PP)"!(ComponentPose=@n19, Alpha=1, ConstraintWeights_0=0.5, ConstraintWeights_1:float=0.5) -> (Pose) {
    alphaBoolBlend {blendOption=Linear}
    boneToModify {boneName=dyn_frontSkirt_01}
    constraintSetup [{targetBone {boneName=dyn_L_skirt_B_01}}, {targetBone {boneName=dyn_R_skirt_B_01}}]
    constraintWeights [1,1]
   }
   n21 = "Constraint - dyn_frontSkirt_02 (PP)"!(ComponentPose=@n20, Alpha=1, ConstraintWeights_0=0.5, ConstraintWeights_1:float=0.5) -> (Pose) {
    alphaBoolBlend {blendOption=Linear}
    boneToModify {boneName=dyn_frontSkirt_02}
    constraintSetup [{targetBone {boneName=dyn_L_skirt_B_02}}, {targetBone {boneName=dyn_R_skirt_B_02}}]
    constraintWeights [1,1]
   }
   n22 = "Constraint - dyn_frontSkirt_03 (PP)"!(ComponentPose=@n21, Alpha=1, ConstraintWeights_0=0.5, ConstraintWeights_1:float=0.5) -> (Pose) {
    alphaBoolBlend {blendOption=Linear}
    boneToModify {boneName=dyn_frontSkirt_03}
    constraintSetup [{targetBone {boneName=dyn_L_skirt_B_03}}, {targetBone {boneName=dyn_R_skirt_B_03}}]
    constraintWeights [1,1]
   }
   n23 = "Constraint - dyn_frontSkirt_04 (PP)"!(ComponentPose=@n22, Alpha=1, ConstraintWeights_0=0.5, ConstraintWeights_1:float=0.5) -> (Pose) {
    alphaBoolBlend {blendOption=Linear}
    boneToModify {boneName=dyn_frontSkirt_04}
    constraintSetup [{targetBone {boneName=dyn_L_skirt_B_04}}, {targetBone {boneName=dyn_R_skirt_B_04}}]
    constraintWeights [1,1]
   }
   n24 = "Constraint - dyn_R_skirt_D_04 (PP)"!(ComponentPose=@n23, Alpha=1, ConstraintWeights_0=0.5, ConstraintWeights_1:float=0.5) -> (Pose) {
    alphaBoolBlend {blendOption=Linear}
    boneToModify {boneName=dyn_R_skirt_D_04}
    constraintSetup [{targetBone {boneName=dyn_R_skirt_C_04}}, {targetBone {boneName=dyn_backSkirt_04}}]
    constraintWeights [1,1]
   }
   n25 = "Transform (Modify) Bone - Bone: dyn_backpack"!(
    ComponentPose=@n24
    bAlphaBoolEnabled:bool=Get EnsembleBackpack() -> (EnsembleBackpack:bool)
    Translation="12,0,0"
    Rotation="0,0,0"
    Scale="1,1,1"
   ) -> (Pose) {
    alphaInputType=Bool
    rotationMode=BMM_Additive
    rotationSpace=BCS_BoneSpace
    scaleMode=BMM_Additive
    scaleSpace=BCS_BoneSpace
    translationMode=BMM_Additive
    translationSpace=BCS_BoneSpace
    alphaBoolBlend {blendOption=Linear}
    boneToModify {boneName=dyn_backpack}
    translation {x=12}
   }
   Output Pose(Result=Component To Local(ComponentPose=@n25)) {layerGroup=None}
   n26 = "Constraint - dyn_L_skirt_B_master (PPPPPP)"!(
    ComponentPose=Local To Component(LocalPose=Copy Pose From Mesh() {bUseAttachedParent=true})
    Alpha=1
    ConstraintWeights_0=0.1593
    ConstraintWeights_1:float=0.4686
    ConstraintWeights_2:float=0.0605
    ConstraintWeights_3:float=0.01
    ConstraintWeights_4:float=0.0842
    ConstraintWeights_5:float=0.2175
   ) -> (Pose) {
    alphaBoolBlend {blendOption=Linear}
    boneToModify {boneName=dyn_L_skirt_B_master}
    constraintSetup [
     {targetBone {boneName=spine_02}}
     {targetBone {boneName=spine_04}}
     {targetBone {boneName=spine_03}}
     {targetBone {boneName=spine_05}}
     {targetBone {boneName=spine_01}}
     {targetBone {boneName=pelvis}}
    ]
    constraintWeights [1,1,1,1,1,1]
   }
   n27 = "Constraint - dyn_R_skirt_B_master (PPPPPP)"!(
    ComponentPose=@n26
    Alpha=1
    ConstraintWeights_0=0.1553
    ConstraintWeights_1:float=0.4402
    ConstraintWeights_2:float=0.0648
    ConstraintWeights_3:float=0.0061
    ConstraintWeights_4:float=0.0865
    ConstraintWeights_5:float=0.247
   ) -> (Pose) {
    alphaBoolBlend {blendOption=Linear}
    boneToModify {boneName=dyn_R_skirt_B_master}
    constraintSetup [
     {targetBone {boneName=spine_02}}
     {targetBone {boneName=spine_04}}
     {targetBone {boneName=spine_03}}
     {targetBone {boneName=spine_05}}
     {targetBone {boneName=spine_01}}
     {targetBone {boneName=pelvis}}
    ]
    constraintWeights [1,1,1,1,1,1]
   }
   n28 = "Constraint - dyn_frontSkirt_master (PPPPPP)"!(
    ComponentPose=@n27
    Alpha=1
    ConstraintWeights_0=0.1828
    ConstraintWeights_1:float=0.4683
    ConstraintWeights_2:float=0.0653
    ConstraintWeights_3:float=0.0013
    ConstraintWeights_4:float=0.0952
    ConstraintWeights_5:float=0.1871
   ) -> (Pose) {
    alphaBoolBlend {blendOption=Linear}
    boneToModify {boneName=dyn_frontSkirt_master}
    constraintSetup [
     {targetBone {boneName=spine_02}}
     {targetBone {boneName=spine_04}}
     {targetBone {boneName=spine_03}}
     {targetBone {boneName=spine_05}}
     {targetBone {boneName=spine_01}}
     {targetBone {boneName=pelvis}}
    ]
    constraintWeights [1,1,1,1,1,1]
   }
   n29 = Control Rig(Source=Component To Local(ComponentPose=@n28)) {
    alphaBoolBlend {blendOption=Linear}
    controlRigAssetReference {blueprintRigClass=&/F_MED_Blue_Jay_CtrlRig.F_MED_Blue_Jay_CtrlRig_C}
   }
   Save cached pose 'presim'(Pose=@n29) {cacheName=presim}
   Get bIsMoving2D()
  ]}
 EventGraph {nodes [
   ResetDynamics(InTeleportType=TeleportPhysics)
   > ResetDynamics(InTeleportType=ResetPhysics)
   Event BlueprintUpdateAnimation()
   > n1 = Sequence!(execute) -> (then_0, then_1, then_2:Exec, then_3:Exec)
   n2 = GetOwnerSkeletalMeshComponent()
   n3 = BreakRotator(InRot=BreakTransform(InTransform=GetDeltaTransformFromRefPose(self=@n2, BoneName=thigh_l, BaseName=pelvis)).Rotation)
   > Set L_thighRotationAlpha(L_thighRotationAlpha:double=MapRangeClamped(Value=@n3.Yaw, InRangeA=-40, InRangeB=-70)) -> (Output_Get:double)
   n4 = BreakTransform(InTransform=GetDeltaTransformFromRefPose(self=@n2, BoneName=thigh_r, BaseName=pelvis))
   > Set R_thighRotationAlpha(R_thighRotationAlpha:double=MapRangeClamped(Value=BreakRotator(InRot=@n4.Rotation).Yaw, InRangeA=-40, InRangeB=-70)) -> (Output_Get:double)
   > Set isDBNO(isDBNO:double="To Float (Boolean)"(InBool=Get bIsDBNO())) -> (Output_Get:double)
   > Set Melee sliding(Melee sliding:bool=false) -> (Output_Get:bool)
   > Set weapon sliding(weapon sliding:bool=false) -> (Output_Get:bool)
   > n5 = Switch on EFortWeaponCoreAnimation!(
    execute
    Selection:EFortWeaponCoreAnimation Enum=GetWeaponCoreAnim!(self:Fort Player Animation Interface Object Reference=GetOwnerAnimationInterface!(self:Custom Character Part Anim Instance Object Reference) -> (ReturnValue:Fort Player Animation Interface Object Reference)) -> (ReturnValue:EFortWeaponCoreAnimation Enum)
   ) -> (
    Melee:Exec
    Pistol:Exec
    Shotgun:Exec
    PaperBlueprint:Exec
    Rifle:Exec
    MeleeOneHand:Exec
    MachinePistol:Exec
    RocketLauncher:Exec
    GrenadeLauncher:Exec
    GoingCommando:Exec
    AssaultRifle:Exec
    TacticalShotgun:Exec
    SniperRifle:Exec
    TrapPlacement:Exec
    ShoulderLauncher:Exec
    AbilityDecoTool:Exec
    Crossbow:Exec
    C4:Exec
    RemoteControl:Exec
    DualWield:Exec
    AR_BullPup:Exec
    AR_ForwardGrip:Exec
    MedPackPaddles:Exec
    SMG_P90:Exec
    AR_DrumGun:Exec
    Consumable_Small:Exec
    Consumable_Large:Exec
    Balloon:Exec
    MountedTurret:Exec
    CreativeTool:Exec
    ExplosiveBow:Exec
    AshtonIndigo:Exec
    AshtonChicago:Exec
    MeleeDualWield:Exec
    Unarmed:Exec
   )
   Branch(
    execute=[
     @n5.AssaultRifle
     @n5.Crossbow
     @n5.DualWield
     @n5.GrenadeLauncher
     @n5.MachinePistol
     @n5.Pistol
     @n5.Rifle
     @n5.RocketLauncher
     @n5.Shotgun
     @n5.SniperRifle
     @n5.TacticalShotgun
    ]
    Condition=Get bIsSliding()
   )
   > Set weapon sliding(weapon sliding=true)
   Branch(execute=[@n5.Melee, @n5.MeleeDualWield], Condition=Get bIsSliding())
   > Set Melee sliding(Melee sliding=true)
   n6 = Branch(
    execute=@n1.then_1
    Condition=OR Boolean!(A=Get bIsMoving2D(), B=Get Melee sliding() -> (Melee sliding:bool), C:bool=Get bIsJumping()) -> (ReturnValue)
   )
   > Set Movement(Movement:bool=true) -> (Output_Get:bool)
   Set Movement(execute=@n6.else, Movement=false)
   Branch(execute=@n1.then_2, Condition=Get weapon sliding() -> (weapon sliding:bool))
   > Set Movement(Movement=false)
   n7 = Branch(
    execute=@n1.then_3
    Condition=BackpackItemDefinitionHasAnyMetaTag!(Tags:Array of Gameplay Tag Structures=Make Array("[0]"=(TagName="Cosmetics.MetaTags.EnsembleMaroon")), bRequiresEquipped:bool=false) -> (ReturnValue:bool)
   )
   > Set EnsembleBackpack(EnsembleBackpack:bool=true) -> (Output_Get:bool)
   Set EnsembleBackpack(execute=@n7.else, EnsembleBackpack=false)
   Delay(Duration=0.3)
   Event OnBeginFalling()
   Event OnEndCrouching()
   Event OnEndPlayingEmote() {is=title K2Node_Event}
   Set L_thighRotationAlpha_extreme(L_thighRotationAlpha_extreme:double=MapRangeClamped(Value=@n3.Yaw, InRangeA=-70, InRangeB=-120)) -> (Output_Get:double)
  ]}
]
variables [
 EnsembleBackpack {type=boolean}
 isDBNO {type=number}
 L_thighAlphaWhenCrouching {type=number}
 L_thighRotationAlpha {type=number}
 L_thighRotationAlpha_extreme {type=number}
 Melee sliding {type=boolean}
 Movement {type=boolean}
 R_thighRotationAlpha {type=number}
 weapon sliding {type=boolean}
]
